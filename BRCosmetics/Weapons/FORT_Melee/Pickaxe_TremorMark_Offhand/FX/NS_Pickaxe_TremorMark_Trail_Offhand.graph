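# Niagara system for the weapon trail of the TremorMark offhand pickaxe, drawn as a single ribbon emitter.
# Fortnite anim-trail modules spawn, initialize and update short-lived particles from bone indices, with trail width taken in FromFirst mode and a ribbon ID offset utility at emitter spawn.
# A Scale Color curve fades the muted blue-purple tint's alpha out early in each particle's brief lifetime.
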NiagaraSystem /BRCosmetics/Weapons/FORT_Melee/Pickaxe_TremorMark_Offhand/FX/NS_Pickaxe_TremorMark_Trail_Offhand
dependencies {
 usedDataInterfaces [&NiagaraDataInterfaceCurve, &NiagaraDataInterfaceSocketReader]
 usedDynamicInputs []
 usedModules [
  &self:SystemSpawnScript.SystemScriptSource.SystemScriptGraph.NiagaraNodeAssignment_1.SetVariables_#1
  &SystemState
 ]
}
emitters {NE_Ribbon_01 {
  emitterSpawnScript [&FN_RibbonIDOffsetUtility()]
  emitterUpdateScript [
   &EmitterState()
   &FN_SpawnAnimTrail(AnimTrailWidthMode=::FromFirst, Spawn Interval=0)
  ]
  particleSpawnScript [
   &FN_InitializeAnimTrailBoneIndecies()
   &FN_InitializeAnimTrail(AnimTrailWidth=2, Color=<rgba 0.2501 0.25 0.5 1>, Lifetime=0.175)
  ]
  particleUpdateScript [
   &SolveForcesAndVelocity(Clamp Velocity=false, Limit Acceleration=false)
   &FN_UpdateAnimTrail() {module=FN_UpdateAnimTrail001}
   &ScaleColor(
    Scale Alpha=&FloatFromCurve(FloatCurve=DataInterface() {propertyValues {Curve {keys [{interpMode=RCIM_Cubic, tangentMode=RCTM_User, time=0, value=1}, {interpMode=RCIM_Cubic, tangentMode=RCTM_User, time=0.35, value=0}]}}})
    ScaleA=true
    ScaleRGB=true
   )
  ]
  rendererClasses [&NiagaraRibbonRendererProperties]
  renderers [{rendererClass=&NiagaraRibbonRendererProperties}]
 }}
summary {
 systemName=NS_Pickaxe_TremorMark_Trail_Offhand
 emitters [{emitterName=NE_Ribbon_01, rendererClasses [&NiagaraRibbonRendererProperties]}]
 userVariables [
  User.bIsFrontEnd {type=NiagaraBool}
  User.bIsFrontEndPreview {type=NiagaraBool}
  User.PhysicalSurface {type=NiagaraInt32}
  User.RibbonIDOffset {type=NiagaraInt32}
 ]
}
